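; Creative_Hover_Platform is the Blueprint for the hover platform item in Creative mode.
; It answers two capability queries with constants: it can be targeted by a linear mover, and it disallows invisible props.
; Its construction script is empty, and the file shows no hovering or movement logic of its own.
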(fn IsTargetableByLinearMover
  (return true))

(fn ShouldInvisiblePropsBeAllowed
  (return false))

(fn ConstructionScript)
